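# Niagara effect for the Basil backpack cosmetic: electric arcs over a flickering green glow, both scaled by the User.Intensity float.
# ElectricArcs bursts a few persistent particles with sprite and mesh renderers, spinning them around Z while Intensity drives ring radius, mesh XY scale and a material parameter.
# Glow adds sprites and a green light, with a waveform plus Intensity modulating alpha and light color.
# Particles outlive their set lifetime rather than expiring.
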NiagaraSystem /BRCosmetics/Accessories/FORT_Backpacks/Backpack_M_MED_Basil/FX/NS_Backpack_Basil
dependencies {usedDataInterfaces [], usedDynamicInputs [], usedModules [&SystemState]}
emitters {
 ElectricArcs {
  emitterUpdateScript [
   &EmitterState(Life Cycle Mode=::NewEnumerator1, Loop Behavior=::NewEnumerator1, Loop Duration Mode=::NewEnumerator1, UseLoopDelay=false)
   &SpawnBurst_Instantaneous(Spawn Count=3, Use Spawn Probability=false, UseLoopCountLimit=false)
  ]
  particleSpawnScript [
   &InitializeParticle(
    Color Mode=::NewEnumerator0
    Lifetime=10
    Mesh Scale Mode=::NewEnumerator3
    Mesh Uniform Scale=0.175
    Position Offset=<7 -10 -2>
    UsePositionOffset=true
    Write Material Random=true
   )
   &InitialMeshOrientation(
    Mesh Axis To Orient=<0 0 1>
    Mesh Orientation Mode=::NewEnumerator2
    Rotation=&RandomRangeVector(Minimum=<0 0 0>, Maximum=<0 0 1>)
    Use Rotation Vector=true
   )
  ]
  particleUpdateScript [
   &ParticleState(Kill Particles When Lifetime Has Elapsed=false, Let Infinitely Lived Particles Die When Emitter Deactivates=true)
   &ShapeLocation(
    MaskSpawnGroup=false
    Offset=<7 -10 -2>
    Offset Mode=::NewEnumerator0
    Override Seed=false
    Ring Radius=&Multiply_Float(A=1.5, B=$User.Intensity)
    Shape Primitive=::NewEnumerator4
    UseOwnerScale=true
   )
   &ScaleMeshSize(Scale Factor=&MakeVector(X=$User.Intensity, Y=$User.Intensity, Z=1))
   &UpdateMeshOrientation(Rotation Rate=&RandomRangeFloat(Minimum=1, Maximum=3), Rotation Vector=<0 0 1>)
   &DynamicMaterialParameters(
    Index 0 Param 1=&Multiply_Float(A=0.3, B=$User.Intensity)
    Index 0 Param 3=10
    Index 0 Param 4=1
    Module.Param0WriteEnabled=true
    Module.Param1WriteEnabled=true
    Module.Param2WriteEnabled=true
    Module.Param3WriteEnabled=true
   )
  ]
  rendererClasses [&NiagaraSpriteRendererProperties, &NiagaraMeshRendererProperties]
  renderers [
   {rendererClass=&NiagaraSpriteRendererProperties}
   {rendererClass=&NiagaraMeshRendererProperties, rendererIndex=1}
  ]
 }
 Glow {
  emitterUpdateScript [
   &EmitterState(Life Cycle Mode=::NewEnumerator1, Loop Behavior=::NewEnumerator1, Loop Duration Mode=::NewEnumerator1, UseLoopDelay=false)
   &SpawnBurst_Instantaneous(Use Spawn Probability=false, UseLoopCountLimit=false)
  ]
  particleSpawnScript [&InitializeParticle(
    Color=<rgba 5 10 0 0.3>
    Lifetime=10
    Position Offset=<0 -10 -2>
    Sprite Size Mode=::NewEnumerator3
    Uniform Sprite Size=30
    UsePositionOffset=true
    Write Material Random=true
   )]
  particleUpdateScript [
   &ParticleState(Kill Particles When Lifetime Has Elapsed=false, Let Infinitely Lived Particles Die When Emitter Deactivates=true)
   /Niagara/Noise/Waveform(
    EnableExponent1=false
    Global Amplitude Scale:NiagaraFloat=&Multiply_Float(A=1, B=$User.Intensity)
    GlobalClampMax=false
    GlobalClampMin=false
    UseInput1=false
    UsePhase1=false
    "Waveform [1]":ENiagara_Waveforms=::NewEnumerator12
    "[1] Frequency"=15
   )
   &FN_Light_Attributes(
    Color=<rgba 0.5 1 0 1>
    Create Color Attribute=true
    Create Exponent Attribute=true
    Create Radius Attribute=true
    Exponent=7
    Radius=130
    Scale Color=&Add_Float(A=&Multiply_Float(A=1.5, B=$User.Intensity), B=$Particles.Waveform.Waveform)
   )
   &FN_Sprite_Attributes(
    Color=&MultiplyLinearColor(
     Linear Color=<rgba 1 2 0 0.1>
     Scale Factor=&MakeVector4(
      X=1
      Y=1
      Z=1
      W=&Add_Float(
       A=&Add_Float(A=0, B=&Multiply_Float(A=$User.Intensity, B=1))
       B=$Particles.Waveform.Waveform
      )
     )
    )
    Create Color Attribute=true
    Create SpriteSize Attribute=true
    SpriteSize=&Multiply_Vector2D_ByFloat(Float=3, Vector2D=$Particles.SpriteSize)
    Write to Module Namespace=true
   )
   &DynamicMaterialParameters(
    Index 0 Param 1=0.5
    Index 0 Param 2=1
    Module.Param0WriteEnabled=true
    Module.Param1WriteEnabled=true
    Module.Param2WriteEnabled=true
    Module.Param3WriteEnabled=true
   ) {enabled=false}
   &ScaleColor(
    Scale Alpha=&Add_Float(
     A=&Add_Float(A=0, B=&Multiply_Float(A=$User.Intensity, B=1))
     B=$Particles.Waveform.Waveform
    )
    ScaleA=true
    ScaleRGB=true
   )
  ]
  rendererClasses [&NiagaraSpriteRendererProperties, &NiagaraLightRendererProperties]
  renderers [
   {rendererClass=&NiagaraSpriteRendererProperties}
   {rendererClass=&NiagaraLightRendererProperties, rendererIndex=1}
   {rendererClass=&NiagaraSpriteRendererProperties, rendererIndex=2}
  ]
 }
}
summary {
 systemName=NS_Backpack_Basil
 emitters [
  {emitterName=Glow, rendererClasses [&NiagaraSpriteRendererProperties, &NiagaraLightRendererProperties]}
  {emitterName=ElectricArcs, rendererClasses [&NiagaraSpriteRendererProperties, &NiagaraMeshRendererProperties]}
 ]
 userVariables [User.Intensity {type=NiagaraFloat, defaultValue=1, description=""}]
}
