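# Physics asset for the M_MED_Basil backpack cosmetic, providing simple collision bodies and joints for its left and right "dyn" block and cup bones.
# Each side has a capsule block surrounded by four thin, angled boxes for the front and back, upper and lower cups.
# Each cup is constrained to its own block with a hinge-like limit, giving range only on Swing1.
# No constraint links the two sides.
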PhysicsAsset /BRCosmetics/Accessories/FORT_Backpacks/Backpack_M_MED_Basil/Meshes/M_MED_Basil_Pack_Physics
bodies {
 dyn_l_block [{shapeType=Capsule, length=3, radius=8, center=<0.2983 -0.0099 -0.6286>, rotation=<rot 0 90 0>}]
 dyn_l_cup_back_low [{shapeType=Box, extentX=2, extentY=12, extentZ=12, shapeName=dyn_l_cup_back_low_box, center=<1.7808 3.3645 1.5537>, rotation=<rot 0 -40 0>}]
 dyn_l_cup_back_up [{shapeType=Box, extentX=2, extentY=12, extentZ=12, shapeName=dyn_l_cup_back_up_box, center=<-1.5108 4.2707 1.5537>, rotation=<rot 0 40 0>}]
 dyn_l_cup_fr_low [{shapeType=Box, extentX=2, extentY=12, extentZ=12, shapeName=dyn_l_cup_fr_low_box, center=<2.2522 -5.6841 1.7094>, rotation=<rot 0 40 0>}]
 dyn_l_cup_fr_up [{shapeType=Box, extentX=2, extentY=12, extentZ=12, shapeName=dyn_l_cup_fr_up_box, center=<-2.3925 -5.8513 1.5537>, rotation=<rot 0 -40 0>}]
 dyn_r_block [{shapeType=Capsule, length=3, radius=8, center=<0.2983 -0.0065 0.6329>, rotation=<rot 0 90 0>}]
 dyn_r_cup_back_low [{shapeType=Box, extentX=2, extentY=12, extentZ=12, shapeName=dyn_r_cup_back_low_box, center=<1.7808 3.3645 0>, rotation=<rot 0 -40 0>}]
 dyn_r_cup_back_up [{shapeType=Box, extentX=2, extentY=12, extentZ=12, shapeName=dyn_r_cup_back_up_box, center=<-1.5108 4.2707 -2.3088>, rotation=<rot 0 40 0>}]
 dyn_r_cup_fr_low [{shapeType=Box, extentX=2, extentY=12, extentZ=12, shapeName=dyn_r_cup_fr_low_box, center=<2.2522 -5.6841 -1.9257>, rotation=<rot 0 40 0>}]
 dyn_r_cup_fr_up [{shapeType=Box, extentX=2, extentY=12, extentZ=12, shapeName=dyn_r_cup_fr_up_box, center=<-2.3925 -5.8513 -2.3088>, rotation=<rot 0 -40 0>}]
}
constraints [
 {bone1Name=dyn_r_cup_fr_up, bone2Name=dyn_r_block, swing1LimitDegrees=90, swing2LimitDegrees=0, twistLimitDegrees=0}
 {bone1Name=dyn_r_cup_fr_low, bone2Name=dyn_r_block, swing1LimitDegrees=90, swing2LimitDegrees=0, twistLimitDegrees=0}
 {bone1Name=dyn_r_cup_back_up, bone2Name=dyn_r_block, swing1LimitDegrees=90, swing2LimitDegrees=0, twistLimitDegrees=0}
 {bone1Name=dyn_r_cup_back_low, bone2Name=dyn_r_block, swing1LimitDegrees=90, swing2LimitDegrees=0, twistLimitDegrees=0}
 {bone1Name=dyn_l_cup_fr_up, bone2Name=dyn_l_block, swing1LimitDegrees=90, swing2LimitDegrees=0, twistLimitDegrees=0}
 {bone1Name=dyn_l_cup_fr_low, bone2Name=dyn_l_block, swing1LimitDegrees=90, swing2LimitDegrees=0, twistLimitDegrees=0}
 {bone1Name=dyn_l_cup_back_up, bone2Name=dyn_l_block, swing1LimitDegrees=90, swing2LimitDegrees=0, twistLimitDegrees=0}
 {bone1Name=dyn_l_cup_back_low, bone2Name=dyn_l_block, swing1LimitDegrees=90, swing2LimitDegrees=0, twistLimitDegrees=0}
]
