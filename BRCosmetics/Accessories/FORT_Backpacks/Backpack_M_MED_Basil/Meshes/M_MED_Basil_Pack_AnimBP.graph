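# Animation Blueprint for the Basil backpack cosmetic, which plays a one-shot animation on player eliminations when the reactive variant is enabled.
# The cached reactive or input pose feeds either RigidBody plus cable AnimDynamics, weighted by a jump/fall alpha, or spring bones on cables and cups, selected by bRigidBodyEnabled.
# A frontend state machine bypasses that physics, and the pack bone gets offset when the mesh name contains "Basil".
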AnimBlueprint /BRCosmetics/Accessories/FORT_Backpacks/Backpack_M_MED_Basil/Meshes/M_MED_Basil_Pack_AnimBP
target_skeleton=&/M_MED_Basil_Pack_Skeleton
graphs [
 AnimGraph {nodes [
   n1 = RigidBody(ComponentPose=Local To Component(LocalPose=Use cached pose 'ReactiveAnim'()), Alpha=Get alpha_jump() -> (alpha_jump:double)) {
    bClampLinearTranslationLimitToRefPose=true
    bOverrideWorldGravity=true
    overridePhysicsAsset=&/M_MED_Basil_Pack_Physics
    alphaBoolBlend {blendOption=Linear}
    componentAppliedLinearAccClamp=<3000 3000 3000>
    componentLinearAccScale=<0.4 0.3 0.3>
    componentLinearVelScale=<0.4 0.3 0.3>
    overrideWorldGravity {z=-980}
   }
   n2 = Anim Dynamics - Bone: dyn_r_cable_up!(ComponentPose=@n1, Alpha=Get alpha_jump(), GravityOverride:Vector="0,0,0", ExternalForce="0,0,0") -> (Pose) {
    angularDampingOverride=0.99
    bLinearSpring=true
    bOverrideLinearDamping=true
    bUseGravityOverride=true
    linearDampingOverride=0.99
    linearSpringConstant=350
    simulationSpace=World
    alphaBoolBlend {blendOption=Linear}
    boundBone {boneName=dyn_r_cable_up}
    componentLinearAccScale=<0.4 0.3 0.3>
    componentLinearVelScale=<0.4 0.3 0.3>
    physicsBodyDefinitions [{
      boundBone {boneName=dyn_r_cable_up}
      boxExtents=<10 10 10>
      constraintSetup {angularLimitsMax=<0 0 0>, angularLimitsMin=<0 0 0>, angularTarget=<0 0 0>, linearAxesMax=<3 0 0>, linearAxesMin=<-3 0 0>}
      localJointOffset=<0.9 0 0>
     }]
    relativeSpaceBone {boneName=dyn_r_block}
   }
   n3 = Anim Dynamics - Bone: dyn_r_cable_low!(ComponentPose=@n2, Alpha=Get alpha_jump(), GravityOverride:Vector="0,0,0", ExternalForce="0,0,0") -> (Pose) {
    angularDampingOverride=0.99
    bLinearSpring=true
    bOverrideLinearDamping=true
    bUseGravityOverride=true
    linearDampingOverride=0.99
    linearSpringConstant=350
    simulationSpace=World
    alphaBoolBlend {blendOption=Linear}
    boundBone {boneName=dyn_r_cable_low}
    componentLinearAccScale=<0.4 0.3 0.3>
    componentLinearVelScale=<0.4 0.3 0.3>
    physicsBodyDefinitions [{
      boundBone {boneName=dyn_r_cable_low}
      boxExtents=<9 2 7>
      constraintSetup {angularLimitsMax=<0 0 0>, angularLimitsMin=<0 0 0>, angularTarget=<0 0 0>, linearAxesMax=<1.6 0 0>, linearAxesMin=<-1.7 0 0>}
      localJointOffset=<0.5 0 0>
     }]
    relativeSpaceBone {boneName=dyn_r_block}
   }
   n4 = Anim Dynamics - Bone: dyn_l_cable_up!(ComponentPose=@n3, Alpha=Get alpha_jump(), GravityOverride:Vector="0,0,0", ExternalForce="0,0,0") -> (Pose) {
    angularDampingOverride=0.99
    bLinearSpring=true
    bOverrideLinearDamping=true
    bUseGravityOverride=true
    linearDampingOverride=0.99
    linearSpringConstant=300
    simulationSpace=World
    alphaBoolBlend {blendOption=Linear}
    alphaScaleBiasClamp {scale=0.8}
    boundBone {boneName=dyn_l_cable_up}
    componentLinearAccScale=<0.4 0.3 0.3>
    componentLinearVelScale=<0.4 0.3 0.3>
    physicsBodyDefinitions [{
      boundBone {boneName=dyn_l_cable_up}
      boxExtents=<4 3 3>
      constraintSetup {angularLimitsMax=<0 0 0>, angularLimitsMin=<0 0 0>, angularTarget=<0 0 0>, linearAxesMax=<2.5 0 0>, linearAxesMin=<-2.5 0 0>}
      localJointOffset=<0.5 0 0>
     }]
    relativeSpaceBone {boneName=dyn_r_block}
   }
   n5 = Anim Dynamics - Bone: dyn_l_cable_low!(ComponentPose=@n4, Alpha=Get alpha_jump(), GravityOverride:Vector="0,0,0", ExternalForce="0,0,0") -> (Pose) {
    angularDampingOverride=0.99
    bLinearSpring=true
    bOverrideLinearDamping=true
    bUseGravityOverride=true
    linearDampingOverride=0.99
    linearSpringConstant=400
    simulationSpace=World
    alphaBoolBlend {blendOption=Linear}
    boundBone {boneName=dyn_l_cable_low}
    componentLinearAccScale=<0.4 0.3 0.3>
    componentLinearVelScale=<0.4 0.3 0.3>
    physicsBodyDefinitions [{
      boundBone {boneName=dyn_l_cable_low}
      boxExtents=<8 4 8>
      constraintSetup {angularLimitsMax=<0 0 0>, angularLimitsMin=<0 0 0>, angularTarget=<0 0 0>, linearAxesMax=<1.4 0 0>, linearAxesMin=<-1.8 0 0>}
      localJointOffset=<0.5 0 0>
     }]
    relativeSpaceBone {boneName=dyn_r_block}
   }
   n6 = Spring controller - Bone: dyn_r_cable_up!(ComponentPose:Component Space Pose Link Structure=Local To Component(LocalPose=Use cached pose 'ReactiveAnim'()), Alpha:float=1) -> (Pose:Component Space Pose Link Structure) {
    is=title AnimGraphNode_SpringBone
    springDamping=10
    springStiffness=450
    alphaBoolBlend {blendOption=Linear}
    alphaScaleBiasClamp {scale=0.09}
    springBone {boneName=dyn_r_cable_up}
   }
   n7 = Spring controller - Bone: dyn_r_cable_low!(ComponentPose:Component Space Pose Link Structure=@n6, Alpha:float=1) -> (Pose:Component Space Pose Link Structure) {
    is=title AnimGraphNode_SpringBone
    springDamping=9
    springStiffness=400
    alphaBoolBlend {blendOption=Linear}
    alphaScaleBiasClamp {scale=0.09}
    springBone {boneName=dyn_r_cable_low}
   }
   n8 = Spring controller - Bone: dyn_l_cable_up!(ComponentPose:Component Space Pose Link Structure=@n7, Alpha:float=1) -> (Pose:Component Space Pose Link Structure) {
    is=title AnimGraphNode_SpringBone
    springDamping=12
    springStiffness=480
    alphaBoolBlend {blendOption=Linear}
    alphaScaleBiasClamp {scale=0.09}
    springBone {boneName=dyn_l_cable_up}
   }
   n9 = Spring controller - Bone: dyn_l_cable_low!(ComponentPose:Component Space Pose Link Structure=@n8, Alpha:float=1) -> (Pose:Component Space Pose Link Structure) {
    is=title AnimGraphNode_SpringBone
    springDamping=8
    springStiffness=460
    alphaBoolBlend {blendOption=Linear}
    alphaScaleBiasClamp {scale=0.09}
    springBone {boneName=dyn_l_cable_low}
   }
   n10 = Spring controller - Bone: dyn_r_cup_fr_up!(ComponentPose:Component Space Pose Link Structure=@n9, Alpha:float=1) -> (Pose:Component Space Pose Link Structure) {
    is=title AnimGraphNode_SpringBone
    springDamping=12
    springStiffness=630
    alphaBoolBlend {blendOption=Linear}
    alphaScaleBiasClamp {scale=0.05}
    springBone {boneName=dyn_r_cup_fr_up}
   }
   n11 = Spring controller - Bone: dyn_r_cup_fr_low!(ComponentPose:Component Space Pose Link Structure=@n10, Alpha:float=1) -> (Pose:Component Space Pose Link Structure) {
    is=title AnimGraphNode_SpringBone
    springDamping=12
    springStiffness=630
    alphaBoolBlend {blendOption=Linear}
    alphaScaleBiasClamp {scale=0.04}
    springBone {boneName=dyn_r_cup_fr_low}
   }
   n12 = Spring controller - Bone: dyn_r_cup_back_up!(ComponentPose:Component Space Pose Link Structure=@n11, Alpha:float=1) -> (Pose:Component Space Pose Link Structure) {
    is=title AnimGraphNode_SpringBone
    springDamping=15
    springStiffness=700
    alphaBoolBlend {blendOption=Linear}
    alphaScaleBiasClamp {scale=0.05}
    springBone {boneName=dyn_r_cup_back_up}
   }
   n13 = Spring controller - Bone: dyn_r_cup_back_low!(ComponentPose:Component Space Pose Link Structure=@n12, Alpha:float=1) -> (Pose:Component Space Pose Link Structure) {
    is=title AnimGraphNode_SpringBone
    springDamping=8
    springStiffness=630
    alphaBoolBlend {blendOption=Linear}
    alphaScaleBiasClamp {scale=0.04}
    springBone {boneName=dyn_r_cup_back_low}
   }
   n14 = Spring controller - Bone: dyn_l_cup_back_low!(ComponentPose:Component Space Pose Link Structure=@n13, Alpha:float=1) -> (Pose:Component Space Pose Link Structure) {
    is=title AnimGraphNode_SpringBone
    springDamping=9
    springStiffness=630
    alphaBoolBlend {blendOption=Linear}
    alphaScaleBiasClamp {scale=0.04}
    springBone {boneName=dyn_l_cup_back_low}
   }
   n15 = Spring controller - Bone: dyn_l_cup_back_up!(ComponentPose:Component Space Pose Link Structure=@n14, Alpha:float=1) -> (Pose:Component Space Pose Link Structure) {
    is=title AnimGraphNode_SpringBone
    springDamping=20
    springStiffness=680
    alphaBoolBlend {blendOption=Linear}
    alphaScaleBiasClamp {scale=0.05}
    springBone {boneName=dyn_l_cup_back_up}
   }
   n16 = Spring controller - Bone: dyn_l_cup_fr_low!(ComponentPose:Component Space Pose Link Structure=@n15, Alpha:float=1) -> (Pose:Component Space Pose Link Structure) {
    is=title AnimGraphNode_SpringBone
    springDamping=8
    springStiffness=610
    alphaBoolBlend {blendOption=Linear}
    alphaScaleBiasClamp {scale=0.04}
    springBone {boneName=dyn_l_cup_fr_low}
   }
   n17 = Spring controller - Bone: dyn_l_cup_fr_up!(ComponentPose:Component Space Pose Link Structure=@n16, Alpha:float=1) -> (Pose:Component Space Pose Link Structure) {
    is=title AnimGraphNode_SpringBone
    springDamping=14
    springStiffness=640
    alphaBoolBlend {blendOption=Linear}
    alphaScaleBiasClamp {scale=0.05}
    springBone {boneName=dyn_l_cup_fr_up}
   }
   n18 = Blend Poses by bool(BlendPose_0=Component To Local(ComponentPose=@n5), BlendPose_1=Component To Local(ComponentPose=@n17), bActiveValue=Get bRigidBodyEnabled()) {blendType=Linear}
   AnimGraphNode_StateMachine_3 = FE_OnKillStateMachine() {is=title AnimGraphNode_StateMachine}
   n19 = Get reactivityInFrontend() -> (reactivityInFrontend:bool)
   n20 = "Transform (Modify) Bone - Bone: pack"!(
    ComponentPose=Local To Component(LocalPose=Blend Poses by bool(BlendPose_0=@AnimGraphNode_StateMachine_3, BlendPose_1=@n18, bActiveValue=@n19) {blendType=Linear})
    bAlphaBoolEnabled:bool=Get isBasil() -> (isBasil:bool)
    Translation="0,-2,0"
    Rotation="0,0,0"
    Scale
   ) -> (Pose) {
    alphaInputType=Bool
    scaleMode=BMM_Additive
    scaleSpace=BCS_BoneSpace
    translationMode=BMM_Additive
    translationSpace=BCS_BoneSpace
    alphaBoolBlend {blendOption=Linear}
    boneToModify {boneName=pack}
    translation {y=-2}
   }
   Output Pose(Result=Component To Local(ComponentPose=@n20)) {layerGroup=None}
   AnimGraphNode_StateMachine_0 = OnKillStateMachine() {is=title AnimGraphNode_StateMachine}
   Save cached pose 'ReactiveAnim'(Pose=Blend Poses by bool(
     BlendPose_0=@AnimGraphNode_StateMachine_0
     BlendPose_1=Input Pose()
     bActiveValue=Get CosmeticVariantReactivity() -> (CosmeticVariantReactivity:bool)
    ) {blendType=Linear}) {cacheName=ReactiveAnim}
  ]}
 OnKillStateMachine {
  entry=Idle
  path=AnimGraph.AnimGraphNode_StateMachine_0.OnKillStateMachine
  ids=id
  nodes [
   AnimStateNode_1 = Idle() {
    stateEntered {guid=#1, notifyName=IdleStateEntered}
    stateFullyBlended {guid=#2}
    stateLeft {guid=#3}
   }
   AnimStateNode_2 = OnKill() {
    is=title AnimStateNode
    bAlwaysResetOnEntry=true
    stateEntered {guid=#4, notifyName=OnKillStateEntered}
    stateFullyBlended {guid=#5}
    stateLeft {guid=#6}
   }
   AnimStateTransitionNode_2 = AnimStateTransitionNode() {crossfadeDuration=0}
   AnimStateTransitionNode_4 = AnimStateTransitionNode() {bAutomaticRuleBasedOnSequencePlayerInState=true}
  ]
  links [
   {
    dst=OnKill
    rule=AnimGraph.AnimGraphNode_StateMachine_0.OnKillStateMachine.AnimStateTransitionNode_2.Transition
    src=Idle
    via=AnimStateTransitionNode_2
   }
   {
    dst=Idle
    rule=AnimGraph.AnimGraphNode_StateMachine_0.OnKillStateMachine.AnimStateTransitionNode_4.Transition
    src=OnKill
    via=AnimStateTransitionNode_4
   }
  ]
 }
 Idle {
  path=AnimGraph.AnimGraphNode_StateMachine_0.OnKillStateMachine.AnimStateNode_1.Idle
  nodes [Output Animation Pose(Result="Sequence Evaluator (by time) 'Basil_Anim'"!(ExplicitTime:float=0) -> (Pose:Pose Link Structure) {is=title AnimGraphNode_SequenceEvaluator, sequence=/BRCosmetics/Accessories/FORT_Backpacks/Backpack_M_MED_Basil/Animations/Basil_Anim})]
 }
 OnKill {
  path=AnimGraph.AnimGraphNode_StateMachine_0.OnKillStateMachine.AnimStateNode_2.OnKill
  nodes [Output Animation Pose(Result=Sequence Player 'Basil_Anim'!(
     Sequence:Anim Sequence Base Object Reference=&Basil_Anim
     bLoopAnimation:bool=False
    ) -> (Pose) {bLoopAnimation=false})]
 }
 Transition {
  path=AnimGraph.AnimGraphNode_StateMachine_0.OnKillStateMachine.AnimStateTransitionNode_2.Transition
  nodes [Result(bCanEnterTransition=Get activeKill() -> (activeKill:bool))]
 }
 Transition {
  path=AnimGraph.AnimGraphNode_StateMachine_0.OnKillStateMachine.AnimStateTransitionNode_4.Transition
  nodes [
   Result(bCanEnterTransition="float < float"(A="Get Relevant Anim Time Remaining (OnKill)"!() -> (ReturnValue:float) {is=title K2Node_AnimGetter}, B=0.1))
   NOT Boolean(A=Get activeKill())
  ]
 }
 FE_OnKillStateMachine {
  entry=FE_Idle
  path=AnimGraph.AnimGraphNode_StateMachine_3.FE_OnKillStateMachine
  ids=id
  nodes [
   AnimStateNode_1 = FE_Idle() {
    is=title AnimStateNode
    stateEntered {guid=#1}
    stateFullyBlended {guid=#2}
    stateLeft {guid=#3}
   }
   AnimStateNode_2 = FE_OnKill() {
    is=title AnimStateNode
    bAlwaysResetOnEntry=true
    stateEntered {guid=#4}
    stateFullyBlended {guid=#5}
    stateLeft {guid=#6}
   }
   AnimStateTransitionNode_2 = AnimStateTransitionNode() {crossfadeDuration=0}
   AnimStateTransitionNode_4 = AnimStateTransitionNode() {bAutomaticRuleBasedOnSequencePlayerInState=true}
  ]
  links [
   {
    dst=FE_OnKill
    rule=AnimGraph.AnimGraphNode_StateMachine_3.FE_OnKillStateMachine.AnimStateTransitionNode_2.Transition
    src=FE_Idle
    via=AnimStateTransitionNode_2
   }
   {
    dst=FE_Idle
    rule=AnimGraph.AnimGraphNode_StateMachine_3.FE_OnKillStateMachine.AnimStateTransitionNode_4.Transition
    src=FE_OnKill
    via=AnimStateTransitionNode_4
   }
  ]
 }
 FE_Idle {
  path=AnimGraph.AnimGraphNode_StateMachine_3.FE_OnKillStateMachine.AnimStateNode_1.FE_Idle
  nodes [Output Animation Pose(Result="Sequence Evaluator (by time) 'Basil_Anim'"(ExplicitTime=0) {sequence=&Basil_Anim})]
 }
 FE_OnKill {
  path=AnimGraph.AnimGraphNode_StateMachine_3.FE_OnKillStateMachine.AnimStateNode_2.FE_OnKill
  nodes [Output Animation Pose(Result=Sequence Player 'Basil_Anim'(Sequence=&Basil_Anim, bLoopAnimation=False) {bLoopAnimation=false})]
 }
 Transition {
  path=AnimGraph.AnimGraphNode_StateMachine_3.FE_OnKillStateMachine.AnimStateTransitionNode_2.Transition
  nodes [Result(bCanEnterTransition=Get activeKill())]
 }
 Transition {
  path=AnimGraph.AnimGraphNode_StateMachine_3.FE_OnKillStateMachine.AnimStateTransitionNode_4.Transition
  nodes [
   Result(bCanEnterTransition="float < float"(A="Get Relevant Anim Time Remaining (OnKill)"(), B=0.1))
   NOT Boolean(A=Get activeKill())
  ]
 }
 EventGraph {nodes [
   n1 = Event BlueprintUpdateAnimation()
   > Set Delta Time X(Delta Time X=@n1.DeltaTimeX)
   > n2 = Sequence!(execute) -> (then_0, then_1, then_2:Exec, then_3:Exec, then_4:Exec, then_5:Exec)
   > n3 = Branch(Condition=OR Boolean(A=Get bIsJumping(), B=Get bIsFalling()))
   > Set alpha_jump(alpha_jump:double=FInterpTo(Current=Get alpha_jump(), Target=1, DeltaTime=Get Delta Time X(), InterpSpeed=2)) -> (Output_Get:double)
   Set alpha_jump(execute=@n3.else, alpha_jump=FInterpTo(Current=Get alpha_jump(), Target=0.3, DeltaTime=Get Delta Time X(), InterpSpeed=2))
   n4 = Cast To FortPlayerPawn(execute=@n2.then_2, Object=TryGetPawnOwner())
   n5 = GetSkeletalMeshForPartType(self=@n4)
   > Set isBasil(isBasil:bool=Contains(SearchIn=GetDisplayName(Object=@n5), Substring=Basil)) -> (Output_Get:bool)
   Set reactivityInFrontend(
    execute=@n2.then_3
    reactivityInFrontend:bool=AND Boolean(A=Get bShouldDisplayFrontendReactivity() -> (bShouldDisplayFrontendReactivity:bool), B=Get CosmeticVariantReactivity())
   ) -> (Output_Get:bool)
   Event OnCharacterCustomizationLoaded()
   > n6 = Sequence!(execute) -> (then_0, then_1, then_2)
   n7 = TryGetFortPawn()
   > n8 = Cast To FortPlayerPawn(Object=@n7)
   > n9 = WatchCosmeticStat!(
    execute:Exec
    Pawn:Fort Player Pawn Object Reference=@n8
    StatTag:Fort Stat Manager Tag Structure=(TagName="StatManager.AthenaKills")
    DefaultValue:int=0
    bBroadcastInitialValue:bool=false
    InitialValueBroadcastDelay:float=-1
    ChangeValueBroadcastDelay:float=-1
    bConsumeWaitingNotifications:bool=true
   ) -> (then:Exec, OnStatChanged:Exec, StatName:Name, StatValue:int) {is=title K2Node_AsyncAction}
   Branch(execute=@n9.OnStatChanged, Condition="integer > integer"(A=@n9.StatValue))
   > Set activeKill(activeKill:bool=true) -> (Output_Get:bool)
   n10 = Cast To FortPlayerPawn(execute=@n6.then_1, Object=@n7)
   > n11 = Cast To B_Athena_PartModifier_Basil_Backpack!(
    execute
    Object=GetModifierForPartType!(self:Fort Player Pawn Object Reference=@n10, PartType:EFortCustomPartType Enum=Backpack) -> (ReturnValue:Custom Character Part Modifier Object Reference)
   ) -> (then, CastFailed, AsB Athena Part Modifier Basil Backpack:B Athena Part Modifier Basil Backpack Object Reference) {is=title K2Node_DynamicCast}
   > Set CosmeticVariantReactivity(CosmeticVariantReactivity:bool=Get Reactivity Enabled!(self:B Athena Part Modifier Basil Backpack Object Reference=@n11) -> (Reactivity Enabled:bool)) -> (Output_Get:bool)
   n12 = Cast To FortPlayerPawn(execute=@n6.then_2, Object=TryGetFortPawn())
   > Set Pawn(Pawn:Fort Player Pawn Object Reference=@n12) -> (Output_Get:Fort Player Pawn Object Reference)
   > n13 = Cast To FortPlayerPawn(Object=GetOwningActor())
   > n14 = GetVariantChannelActiveIndexForItemActor!(
    execute
    Item:Athena Cosmetic Item Definition Object Reference=/BRCosmetics/Athena/Items/Cosmetics/Backpacks/Backpack_Basil
    Actor:Actor Object Reference=@n13
    Channel:Gameplay Tag Structure=(TagName="Cosmetics.Variant.Channel.Parts")
    VariantList:Array of Gameplay Tag Structures=Make Array("[0]"=(TagName="Cosmetics.Variant.Property.Stage1"))
    bErrorIfChannelMissing:bool=false
    bErrorIfTagNotInArray:bool=false
   ) -> (then, Index:int=0, bIndexIsValid:bool, bFoundChannel:bool)
   > n15 = Branch(Condition=@n14.bIndexIsValid)
   > Set CosmeticVariantReactivity(CosmeticVariantReactivity=true)
   Set CosmeticVariantReactivity(execute=@n15.else, CosmeticVariantReactivity=false)
   AnimNotify_IdleStateEntered() {is=title K2Node_Event}
   > Set activeKillAnimEnd(activeKillAnimEnd:bool=false) -> (Output_Get:bool)
   AnimNotify_OnKillStateEntered() {is=title K2Node_Event}
   n16 = Get KillAnimation() -> (KillAnimation:Anim Sequence Base Object Reference)
   > RetriggerableDelay(Duration=float + float(
     A=GetPlayLength!(self:Animation Asset Object Reference=@n16) -> (ReturnValue:float)
     B=Get KillAnimationAdditionalTime() -> (KillAnimationAdditionalTime:double)
    ))
   > Set activeKillAnimEnd(activeKillAnimEnd=true)
   n17 = Event BlueprintProcessAnimInputEvent()
   > Branch(Condition=IsNamed(self=@n17.AnimInputEvent, InEventName=aie.EnemyPlayerEliminated))
   > RetriggerableDelay(Duration=0.6)
   > Set activeKill(activeKill=true)
   Event BlueprintPostEvaluateAnimation() {is=title K2Node_Event}
   > Set activeKill(activeKill=false)
  ]}
]
variables [
 "0,05 * Alpha" {type=number, default=1}
 activeKill {type=boolean}
 activeKillAnimEnd {type=boolean}
 activeKillAnimEnd_0 {type=boolean}
 alpha_jump {type=number}
 CosmeticVariantReactivity {type=boolean}
 Delta Time X {type=number}
 isBasil {type=boolean}
 KillAnimation {type=AnimSequenceBase, default=&Basil_Anim}
 KillAnimationAdditionalTime {type=number, default=0.25}
 Pawn {type=FortPlayerPawn}
 reactivityInFrontend {type=boolean}
]
